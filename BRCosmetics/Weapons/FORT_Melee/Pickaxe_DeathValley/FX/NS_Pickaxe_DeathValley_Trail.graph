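# Niagara system for the Death Valley pickaxe's swing trail, built from two short-lived ribbon emitters, TrailPickaxe and TrailSkull.
# Both use shared FN anim-trail modules to spawn, initialize and update ribbons.
# TrailPickaxe is near-black, while TrailSkull is translucent white and uses the FromSecond trail-width mode.
# Each fades alpha over life through a steep-falloff curve and tints RGB by a shared System.TOD color.
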NiagaraSystem /BRCosmetics/Weapons/FORT_Melee/Pickaxe_DeathValley/FX/NS_Pickaxe_DeathValley_Trail
dependencies {
 usedDataInterfaces [&NiagaraDataInterfaceCurve, &NiagaraDataInterfaceSocketReader, &NiagaraDataInterfaceColorCurve]
 usedDynamicInputs [&ColorFromCurve]
 usedModules [
  &self:SystemSpawnScript.SystemScriptSource.SystemScriptGraph.NiagaraNodeAssignment_0.SetVariables_#1
  &self:SystemSpawnScript.SystemScriptSource.SystemScriptGraph.NiagaraNodeAssignment_2.SetVariables_#2
  &SystemState
 ]
}
emitters {
 TrailPickaxe {
  emitterUpdateScript [
   &EmitterState()
   &FN_SpawnAnimTrail()
  ]
  particleSpawnScript [
   &FN_InitializeAnimTrailBoneIndecies() {module=FN_InitializeAnimTrailBoneIndecies001}
   &FN_InitializeAnimTrail(AnimTrailWidth=0.6, Color=<rgba 0.01 0.01 0.01 1>, Lifetime=0.23) {module=FN_InitializeAnimTrail001}
  ]
  particleUpdateScript [
   &ScaleColor(
    Scale Alpha=&FloatFromCurve(
     FloatCurve=DataInterface() {propertyValues {Curve {keys [
         {interpMode=RCIM_Cubic, leaveTangent=-3.2813, tangentMode=RCTM_Break, time=0, value=1}
         {arriveTangent=-0.0411, interpMode=RCIM_Cubic, tangentMode=RCTM_Break, time=1, value=0}
        ]}}}
     Scale Curve=2
    )
    Scale RGB=&MakeVectorFromLinearColorRGB(Color=$System.TOD)
    ScaleA=true
    ScaleRGB=true
   ) {module=ScaleColor002}
   &FN_UpdateAnimTrail()
  ]
  rendererClasses [&NiagaraRibbonRendererProperties]
  renderers [{rendererClass=&NiagaraRibbonRendererProperties}]
 }
 TrailSkull {
  emitterUpdateScript [
   &EmitterState()
   &FN_SpawnAnimTrail(AnimTrailWidthMode=::FromSecond)
  ]
  particleSpawnScript [
   &FN_InitializeAnimTrailBoneIndecies() {module=FN_InitializeAnimTrailBoneIndecies001}
   &FN_InitializeAnimTrail(AnimTrailWidth=0.6, Color=<rgba 1 1 1 0.5>, Lifetime=0.23) {module=FN_InitializeAnimTrail001}
  ]
  particleUpdateScript [
   &ScaleColor(
    Scale Alpha=&FloatFromCurve(FloatCurve=DataInterface() {propertyValues {Curve {keys [
         {interpMode=RCIM_Cubic, leaveTangent=-3.2813, tangentMode=RCTM_Break, time=0, value=1}
         {arriveTangent=-0.0411, interpMode=RCIM_Cubic, tangentMode=RCTM_Break, time=1, value=0}
        ]}}})
    Scale RGB=&MakeVectorFromLinearColorRGB(Color=$System.TOD)
    ScaleA=true
    ScaleRGB=true
   ) {module=ScaleColor002}
   &FN_UpdateAnimTrail()
  ]
  rendererClasses [&NiagaraRibbonRendererProperties]
  renderers [{rendererClass=&NiagaraRibbonRendererProperties}]
 }
}
summary {
 systemName=NS_Pickaxe_DeathValley_Trail
 emitters [
  {emitterName=TrailSkull, rendererClasses [&NiagaraRibbonRendererProperties]}
  {emitterName=TrailPickaxe, rendererClasses [&NiagaraRibbonRendererProperties]}
 ]
 userVariables [
  User.bIsFrontEnd {type=NiagaraBool}
  User.bIsFrontEndPreview {type=NiagaraBool}
  User.RibbonIDOffset {type=NiagaraInt32}
 ]
}
